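# Additive unlit fog material for the Helios throne skull effect, built for Niagara mesh particles.
# Two noise textures, distorted by a panning radial swirl normal with per-particle offsets, are tinted by a Fresnel two-color blend, hue shift and particle color to form emissive.
# Opacity combines inverse Fresnel, vertex-blue masking and depth fade to soften edges.
# World position offset mixes swirl noise and layered sine sway, masked along V, plus camera offset.
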Material /Game/Effects/Fort_Effects/Materials/Helios/M_Helios_Throne_Skull_Fog
graph {
 properties {
  allowTranslucentCustomDepthWrites=true
  bUsedWithNiagaraMeshParticles=true
  blendMode=BLEND_Additive
  pixelDepthOffsetMode=PDOM_Legacy
  shadingModel=MSM_Unlit
  usedShadingModels=Unlit
 }
 nodes [
  n1 = Constant() {r=1}
  n2 = Fresnel() {exponent=2}
  n3 = OneMinus(None=@n2)
  n4 = Multiply(A=ScalarParameter() {defaultValue=0.5, parameterName=Opactiy}, B=Clamp(None=Multiply(A=@n1, B=@n3)))
  n5 = OneMinus(None=VertexColor().B)
  n6 = ScalarParameter() {defaultValue=10, parameterName=FadeDistance}
  Clamp(None=Multiply(A=Multiply(A=@n4, B=@n5), B=DepthFade(FadeDistance=@n6)))
  Constant() {r=1}
  n7 = MaterialFunctionCall(Float4=Constant4Vector() {constant=<rgba 1 1 1 1>}.RGBA) {materialFunction=&BreakOutFloat4Components}
  n8 = MaterialFunctionCall() {materialFunction=&VectorToRadialValue}
  n9 = MaterialFunctionCall(Float2=@n8.Radial Coordinates) {materialFunction=&BreakOutFloat2Components}
  n10 = Multiply(
   A=Subtract(A=Multiply(A=@n9.G), B=ScalarParameter() {parameterName=MaskOffset})
   B=ScalarParameter() {defaultValue=0.5, parameterName=MaskScale}
  )
  n11 = VectorParameter() {parameterName=Speed, defaultValue=<rgba 0.002 -0.02 0.5 2>}
  n12 = AppendVector(A=@n11.B, B=@n11.B)
  n13 = TextureSample(UVs=Panner(
    Coordinate=MaterialFunctionCall(UVs=@n8.Radial Coordinates, Texture Scale=@n12) {materialFunction=&ScaleUVsByCenter}.UVs
    Speed=AppendVector(A=@n11.R, B=@n11.R)
   )) {samplerType=SAMPLERTYPE_Normal, texture=/Game/Global/Textures/Default/Noise/T_SwirlNoise_N}
  n14 = AppendVector(A=@n13.R, B=@n13.R)
  n15 = ScalarParameter() {defaultValue=2, parameterName=Noise2}
  n16 = Add(A=TextureCoordinate(), B=Multiply(A=Saturate(None=@n10), B=Multiply(A=@n14, B=@n15)))
  n17 = Add(A=Multiply(A=AppendVector(A=@n7.R, B=@n7.R), B=@n16), B=LinearInterpolate(Alpha=ParticleRandom()) {constB=0.3})
  n18 = Saturate(None=DynamicParameter() {paramNames [NoiseAmount,WPOAmount,Emmisive,Param4]}.NoiseAmount)
  n19 = Add(A=Multiply(A=AppendVector(A=@n7.B, B=@n7.B), B=@n16), B=LinearInterpolate(Alpha=ParticleRandom()) {constB=0.2})
  n20 = Add(
   A=TextureSampleParameter2D(UVs=LinearInterpolate(A=@n17, B=@n16, Alpha=@n18)) {parameterName=Texture1, texture=/Game/Effects/Fort_Effects/Textures/Tile/T_Noise_02}.RGB
   B=TextureSampleParameter2D(UVs=LinearInterpolate(A=@n19, B=@n16, Alpha=@n18)) {parameterName=Texture2, texture=&T_Noise_02}.RGB
  )
  n21 = LinearInterpolate(
   A=VectorParameter() {parameterName=Color, defaultValue=<rgba 0.0036 0.099 0.0797 1>}.RGB
   B=VectorParameter() {parameterName=Color2, defaultValue=<rgba 0.2031 1 0.4146 1>}.RGB
   Alpha=@n2
  )
  n22 = Multiply(
   A=MaterialFunctionCall(Hue Shift Percentage=ScalarParameter() {parameterName="Hue Shift Percentage (S)"}, Texture=Multiply(A=@n20, B=@n21)) {materialFunction=&HueShift}.Result
   B=ParticleColor().RGB
  )
  n23 = Multiply(A=ScalarParameter() {defaultValue=1, parameterName=Emmisive}, B=@n22)
  n24 = Clamp(None=Power(Base=@n22, Exp=ScalarParameter() {defaultValue=0.4, parameterName=Exp}) {constExponent=4})
  MaterialFunctionCall(Emissive=Desaturation(None=Add(A=@n23, B=@n24), Fraction=ScalarParameter() {defaultValue=0.25, parameterName=Fraction})) {materialFunction=&MF_FX_EmissiveAdjustments}
  Multiply()
  n25 = Constant()
  n26 = AppendVector(A=@n14, B=@n13.B)
  n27 = ScalarParameter() {defaultValue=2, parameterName=WPO_Texture_Amount}
  n28 = VertexColor()
  n29 = Time()
  n30 = Multiply(A=Multiply(A=Sine(None=@n29) {period=6}, B=Constant() {r=1}), B=Constant3Vector() {constant=<rgba 5 -5 0 0>})
  n31 = Multiply(A=Sine(None=Time()) {period=4}, B=Constant3Vector() {constant=<rgba 1 5 0 0>})
  n32 = Sine(None=Time()) {period=3}
  n33 = Add(
   A=Multiply(A=Multiply(A=@n26, B=@n27), B=OneMinus(None=@n28.B))
   B=Multiply(A=VectorParameter() {parameterName=WPOAmount, defaultValue=<rgba 0.1 0.1 1 1>}.RGB, B=LinearInterpolate(A=@n30, B=@n31, Alpha=@n32))
  )
  n34 = OneMinus(None=ComponentMask(None=TextureCoordinate()) {g=true})
  n35 = ScalarParameter() {parameterName=CameraOffset}
  NamedRerouteDeclaration(None=Add(
    A=LinearInterpolate(A=@n25, B=@n33, Alpha=@n34)
    B=MaterialFunctionCall(Offset Amount=@n35) {materialFunction=/Engine/Functions/Engine_MaterialFunctions02/WorldPositionOffset/CameraOffset}.Clamped Camera Offset
   )) {name=WPO, nodeColor=<rgba 1 0.2812 0 1>}
  NamedRerouteUsage()
  Saturate()
 ]
}
